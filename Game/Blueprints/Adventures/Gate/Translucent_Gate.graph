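# Unlit, modulate-blended translucent material for the adventure Gate, carrying damage-crack and hit effects on the GateDevice physical material.
# The Fortnite_HitEffects function, with crack damage shading enabled, supplies diffuse and hit-and-shrink world position offset from deformation center and damage radius parameters.
# A small vertex-normal push is added to that offset.
# A clamped pixel-depth term lerps the diffuse toward white, which is neutral under modulate blending.
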Material /Game/Blueprints/Adventures/Gate/Translucent_Gate
graph {
 properties {
  blendMode=BLEND_Modulate
  physMaterial=/Game/FortressPhysicalMaterials/GateDevice
  pixelDepthOffsetMode=PDOM_Legacy
  shadingModel=MSM_Unlit
  translucencyPass=MTP_BeforeDOF
  usedShadingModels=Unlit
 }
 nodes [
  n1 = MaterialFunctionCall!(
   WSDeformationCenter=VectorParameter() {parameterName=WSDeformationCenter}.RGB
   Code_DestructionAmount=ScalarParameter() {group=z_code, parameterName=Code_DestructionAmount}
   Code_UseUvForDamage=StaticBool()
   Code_UseHitEffects=StaticBoolParameter() {defaultValue=true, group=z_code, parameterName=Code_UseHitEffects}
   Code_Emissive
   WSDamageRadius=ScalarParameter() {parameterName=WSDamageRadius}
   CrackDamageShader=StaticBool() {value=true}
   Diffuse=Constant() {r=1}
   ScalarModDamageCracks=Constant() {r=1}
   ScalarOffsetforDamageCracks
   ExposeLerpFunctionForInvertingTheHitDistanceCalc
   InvertDamageFalloff
   CrackWorldPosition
   WorldPosition
   DistanceCullFadeAnimationOffset
  ) {materialFunction=/Game/Packages/jl_tests/Fortnite_HitEffects}
  MaterialExpressionWorldNormal_0 = VertexNormalWS()
  Add(A=@n1.Hit And Shrink WPO, B=Multiply(A=@MaterialExpressionWorldNormal_0, B=Constant() {r=0.0125}))
  n2 = Constant() {r=32}
  n3 = Subtract(A=Subtract(B=PixelDepth().R), B=@n2)
  LinearInterpolate(A=@n1.Diffuse, B=Constant() {r=1}, Alpha=Clamp(None=Divide(A=@n3, B=@n2)))
  ScalarParameter() {group=z_code, parameterName=Code_HitAnimation}
  VectorParameter() {group=z_code, parameterName=Code_DeformationVector, defaultValue=<rgba 0 1 0 1>}
 ]
}
